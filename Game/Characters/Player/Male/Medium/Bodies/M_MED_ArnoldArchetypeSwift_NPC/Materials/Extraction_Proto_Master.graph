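# Prototype master material in the Arnold archetype Swift NPC body folder, flagged for skeletal mesh and mesh deformer use.
# It uses Default Lit shading and exposes scalar parameters for metallic, specular and roughness, plus vector parameters for diffuse and emissive color.
# Diffuse defaults to flat dark grey and emissive to black, with no textures or math nodes in the graph.
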Material /Game/Characters/Player/Male/Medium/Bodies/M_MED_ArnoldArchetypeSwift_NPC/Materials/Extraction_Proto_Master
graph {
 properties {bUsedWithMeshDeformer=true, bUsedWithSkeletalMesh=true, pixelDepthOffsetMode=PDOM_Legacy, usedShadingModels=Default Lit}
 nodes [
  ScalarParameter() {parameterName=Metallic}
  ScalarParameter() {parameterName=Spec}
  ScalarParameter() {parameterName=Rough}
  VectorParameter() {parameterName=Diffuse, defaultValue=<rgba 0.224 0.224 0.224 1>}
  VectorParameter() {parameterName=Emissive, defaultValue=<rgba 0 0 0 1>}
 ]
}
